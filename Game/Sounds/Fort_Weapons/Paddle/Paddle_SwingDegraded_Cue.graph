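# Sound cue for the swing of a degraded paddle melee weapon, layering a paddle sword whoosh with a generic melee degrade sound.
# A quality switch selects between two mixers, the second drawing on fewer wave variations.
# In each mixer, both layers pick a random wave and pass through modulators, with the whoosh pitched down to a fixed lower value.
# Small melee attenuation, a melee sound class and player melee concurrency govern playback.
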SoundCue /Game/Sounds/Fort_Weapons/Paddle/Paddle_SwingDegraded_Cue
root=QualityLevel_0
nodes [
 {id=Mixer_1, children [Modulator_3,Modulator_2], properties {inputVolume [1,1]}}
 {id=Mixer_2, children [Modulator_5,Modulator_4], properties {inputVolume [1,1]}}
 {id=Modulator_2, children [Random_2]}
 {id=Modulator_3, children [Random_3], properties {pitchMax=0.8, pitchMin=0.8}}
 {id=Modulator_4, children [Random_4]}
 {id=Modulator_5, children [Random_5], properties {pitchMax=0.8, pitchMin=0.8}}
 {id=QualityLevel_0, children [Mixer_1,Mixer_2,null,null]}
 {id=Random_2, children [WavePlayer_10, WavePlayer_11, WavePlayer_12, WavePlayer_13], properties {weights [1,1,1,1]}}
 {
  id=Random_3
  children [WavePlayer_14, WavePlayer_15, WavePlayer_16, WavePlayer_17, WavePlayer_18, WavePlayer_19]
  properties {weights [1,1,1,1,1,1]}
 }
 {id=Random_4, children [WavePlayer_20, WavePlayer_21], properties {weights [1,1]}}
 {id=Random_5, children [WavePlayer_24, WavePlayer_25, WavePlayer_26, WavePlayer_27], properties {weights [1,1,1,1]}}
 {id=WavePlayer_10, properties {soundWaveAssetPtr=/Game/Sounds/Fort_Weapons/Degrades/degrade_melee_generic_01}}
 {id=WavePlayer_11, properties {soundWaveAssetPtr=/Game/Sounds/Fort_Weapons/Degrades/degrade_melee_generic_02}}
 {id=WavePlayer_12, properties {soundWaveAssetPtr=/Game/Sounds/Fort_Weapons/Degrades/degrade_melee_generic_03}}
 {id=WavePlayer_13, properties {soundWaveAssetPtr=/Game/Sounds/Fort_Weapons/Degrades/degrade_melee_generic_04}}
 {id=WavePlayer_14, properties {soundWaveAssetPtr=&/PaddleSword_Whoosh_001}}
 {id=WavePlayer_15, properties {soundWaveAssetPtr=&/PaddleSword_Whoosh_002}}
 {id=WavePlayer_16, properties {soundWaveAssetPtr=&/PaddleSword_Whoosh_003}}
 {id=WavePlayer_17, properties {soundWaveAssetPtr=&/PaddleSword_Whoosh_004}}
 {id=WavePlayer_18, properties {soundWaveAssetPtr=&/PaddleSword_Whoosh_005}}
 {id=WavePlayer_19, properties {soundWaveAssetPtr=&/PaddleSword_Whoosh_006}}
 {id=WavePlayer_20, properties {soundWaveAssetPtr=&degrade_melee_generic_01}}
 {id=WavePlayer_21, properties {soundWaveAssetPtr=&degrade_melee_generic_02}}
 {id=WavePlayer_24, properties {soundWaveAssetPtr=&PaddleSword_Whoosh_001}}
 {id=WavePlayer_25, properties {soundWaveAssetPtr=&PaddleSword_Whoosh_002}}
 {id=WavePlayer_26, properties {soundWaveAssetPtr=&PaddleSword_Whoosh_003}}
 {id=WavePlayer_27, properties {soundWaveAssetPtr=&PaddleSword_Whoosh_004}}
]
properties {
 attenuationSettings=&Fort_Weapon_Melee_Sml_Att
 maxDistance=2800
 soundClassObject=&Fort_WeaponMelee_PSM
 volumeMultiplier=0.9
 concurrencySet [&Player_Melee_Weapon_Conc]
}
